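; Character part modifier that drives Tomato repulsor hand effects for muzzle, idle glow and firing.
; On part change it attaches effects to hand sockets and multicasts a bind to the repulsor weapon's hand-fired event, with a delay hack for laggy clients.
; Without an active power, firing flashes that hand's muzzle, briefly swaps idle effect and emissive color, and shakes the camera.
; Hiding props toggle only the right idle effect; OnRep still casts to an Ashton Indigo weapon.
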(event EventOnPawnAndPartChanged
  ;[1 1 1] New
  (|CallParentFunction)
  (CallFunction|MuzzleAttachment)
  (CallFunction|IdleAttachment)
  (CallFunction|MuzzleFiringSetup 0))

(event Custom|Idle%20Attachment
  (bind _returnvalue (CustomCharacter|GetCharacterPartType))
  (bind _returnvalue_2 (CustomCharacter|GetAssociatedPlayerPawn))
  (bind _returnvalue_3 (Customization|GetSkeletalMeshforPartType _returnvalue_2 _returnvalue :!class /Script/FortniteGame.FortPlayerPawn))
  (Transformation|AttachComponentToComponent (Variables|Default|GetIdle_Left) _returnvalue_3 "FX_repulsor_l" "KeepRelative" "KeepRelative" "KeepRelative" true)
  (Transformation|AttachComponentToComponent (Variables|Default|GetIdle_Right) _returnvalue_3 "FX_repulsor_r" "KeepRelative" "KeepRelative" "KeepRelative" true)
  (Effects|Components|ParticleSystem|SetEmitterEnable (Variables|Default|GetIdle_Right) "Booster" false)
  ;[1 1 1] Cancel Auto Activate of Idle FX, add set active to fix JIRA FORT-308838
  (CallFunction|IsHiddinginProp (GameplayTags|HasMatchingGameplayTag (CustomCharacter|GetAssociatedPlayerPawn) "(TagName=\"Granted.Athena.EnvItem.HidingProp.HidingInProp\")")))

(event Custom|Muzzle_Firing_Multicast
  (params (Weapon :type Fort%20Weapon%20Object%20Reference))
  ;[1 0.6369 0.427] Muzzle Firing Setup - Must be Replicated to Function!
  (bind _asb_high_tower_tomato_repulsors (Utilities|Casting|CastToB_HighTower_Tomato_Repulsors (Variables|/Script/FortniteGame.FortPawn|GetCurrentWeapon (CustomCharacter|GetAssociatedPlayerPawn)))
    (:then
      _)
    (:CastFailed
      (Development|PrintString "CASTFAILED" true true "(R=1.000000,G=0.000000,B=0.031557,A=1.000000)")
      ;[1 0 0] Hack loop to fix current weapon not being valid yet on laggy clients.
      (Utilities|FlowControl|DoN 5
        (:Exit
          (Utilities|FlowControl|Delay 0.1)))))
  (RangedWeaponDual|BindEventtoOnHandFired _asb_high_tower_tomato_repulsors (EventDispatchers|CreateEvent :!event Custom|Firing))
  (Development|PrintString (Utilities|GetDisplayName _asb_high_tower_tomato_repulsors) true true "(R=0.290283,G=1.000000,B=0.000000,A=1.000000)"))

(event Custom|Muzzle_Firing_Setup
  (params (WeaponObject :type Fort%20Weapon%20Object%20Reference))
  ;[1 0.6369 0.427] Muzzle Firing Setup - Must be Replicated to Function!
  (CallFunction|MuzzleFiringMulticast 0))

(event Custom|Firing
  (params (Hand :type EDualWeaponHand%20Enum :default "LEFT") (bPersistantFire :type Boolean))
  (bind _returnvalue (CustomCharacter|GetCharacterPartType))
  (bind _returnvalue_2 (CustomCharacter|GetAssociatedPlayerPawn))
  (bind _returnvalue_3 (Customization|GetSkeletalMeshforPartType _returnvalue_2 _returnvalue :!class /Script/FortniteGame.FortPlayerPawn))
  (bind _idle_right (Variables|Default|GetIdle_Right))
  (bind _idle_left (Variables|Default|GetIdle_Left))
  (bind _lmb_color (Variables|Default|GetLMB%20Color))
  (bind _boostcolor (Variables|Default|GetBoostColor))
  ;[1 0.6369 0.427] Firing - IdleControl
  (if (not (GameplayTags|HasMatchingGameplayTag (CustomCharacter|GetAssociatedPlayerPawn) "(TagName=\"Granted.Athena.Power.ActivePower.IsActive\")"))
    (Utilities|FlowControl|Sequence
      (:then_0
        (if (Utilities|Enum|Equal(Enum) Hand "LEFT")
          (Components|Activation|Activate (Variables|Default|GetMuzzle_Left) false)
          (Rendering|Material|SetVectorParameterValueonMaterials _returnvalue_3 "Emissive Color" (Math|Conversions|ToVector(LinearColor) _lmb_color))
          (Development|SetHiddeninGame _idle_left true false)
          (Utilities|FlowControl|Delay 0.3)
          (Development|SetHiddeninGame _idle_left false false)
          (Components|Activation|Activate _idle_left true)
          (Rendering|Material|SetVectorParameterValueonMaterials _returnvalue_3 "Emissive Color" (Math|Conversions|ToVector(LinearColor) _boostcolor))
          (else
            (Components|Activation|Activate (Variables|Default|GetMuzzle_Right) false)
            (Rendering|Material|SetVectorParameterValueonMaterials _returnvalue_3 "Emissive Color" (Math|Conversions|ToVector(LinearColor) _lmb_color))
            (Development|SetHiddeninGame _idle_right true false)
            (Utilities|FlowControl|Delay 0.3)
            (Development|SetHiddeninGame _idle_right false false)
            (Components|Activation|Activate _idle_right true)
            (Rendering|Material|SetVectorParameterValueonMaterials _returnvalue_3 "Emissive Color" (Math|Conversions|ToVector(LinearColor) _boostcolor)))))
      (:then_1
        ;[1 1 1] Camera shake
        (CameraShakes|StartCameraShake :self (Variables|/Script/Engine.PlayerController|GetPlayerCameraManager (Fortnite|GetFortPlayerControllerfromActor _returnvalue_2)) :ShakeClass "/HighTower/Items/Tomato/Repulsors/Proto/HighTower_Tomato_Repulsor_CameraShake.HighTower_Tomato_Repulsor_CameraShake_C" :Scale 1.0 :PlaySpace "CameraLocal")))))

(event Custom|Muzzle_Attachment
  (bind _returnvalue (CustomCharacter|GetSkeletalMeshForPartOnPlayerPawn))
  ;[1 0.6369 0.427] Muzzle Setups
  (Transformation|AttachComponentToComponent (Variables|Default|GetMuzzle_Left) _returnvalue "FX_repulsor_l" "KeepRelative" "KeepRelative" "KeepRelative" true)
  (Transformation|AttachComponentToComponent (Variables|Default|GetMuzzle_Right) _returnvalue "FX_repulsor_r" "KeepRelative" "KeepRelative" "KeepRelative" true))

(event EventEnteringHidingProp
  (|CallParentFunction)
  (Components|Activation|SetActive (Variables|Default|GetIdle_Right) false true))

(event EventExitingHidingProp
  (|CallParentFunction)
  (Components|Activation|SetActive (Variables|Default|GetIdle_Right) true true))

(fn OnRep_OurCurrentWeapon
  (bind _asb_ranged_ashton_indigo_athena (Utilities|Casting|CastToB_Ranged_Ashton_Indigo_Athena (Variables|Default|GetOurCurrentWeapon))
    (:then
      (RangedWeaponDual|BindEventtoOnHandFired _asb_ranged_ashton_indigo_athena (EventDispatchers|CreateEvent :!event Firing))
      (Development|PrintString (Utilities|GetDisplayName _asb_ranged_ashton_indigo_athena) true true "(R=0.290283,G=1.000000,B=0.000000,A=1.000000)"))
    (:CastFailed
      (Development|PrintString "CASTFAILED" true true "(R=1.000000,G=0.000000,B=0.031557,A=1.000000)"))))

(fn ConstructionScript)
